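# Editor-utility actor component that backs up an actor's static mesh materials so a Season 22 "chrome" look can be undone.
# StoreMaterials caches material lists per owner component, skipping editor-only, already cached or "lackluster"-based meshes.
# Dynamic material instances are stored as their parent material, and debug prints trace the process.
# Unchrome reapplies cached materials by slot, then destroys the component; applying chrome happens outside this file.
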EditorUtilityBlueprint /Game/Blueprints/Blutilities/Developers/brycelumpkin/Seasonal/S22/Chrome_Editor_Comp/Chrome_Editor_Comp
graphs [
 AddComponentToList {nodes [
   n1 = AddComponentToList!() -> (then, meshComponent:"Static Mesh Component Object Reference (by ref)") {is=title K2Node_FunctionEntry}
   > n2 = Branch(Condition=Find(
     TargetMap:Map of Static Mesh Component Object References to Chrome Mat Storage Struct Structures=Get cache() -> (cache:Map of Static Mesh Component Object References to Chrome Mat Storage Struct Structures)
     Key:"Static Mesh Component Object Reference (by ref)"=@n1
    ) -> (Value:Chrome Mat Storage Struct Structure).ReturnValue)
   n3 = Sequence(execute=@n2.else)
   > n4 = For Each Loop(Array:Array of Material Interface Object References=GetMaterials!(self:Mesh Component Object Reference=@n1) -> (ReturnValue:Array of Material Interface Object References)) -> (Array Element:Material Interface Object Reference)
   > n5 = Cast To MaterialInstanceDynamic!(execute, Object=@n4) -> (then, CastFailed, AsMaterial Instance Dynamic:Material Instance Dynamic Object Reference) {is=title K2Node_DynamicCast}
   n6 = Get l_mats() -> (l_mats:Array of Material Interface Object References)
   > Add(
    TargetArray:Array of Material Interface Object References=@n6
    NewItem:"Material Interface Object Reference (by ref)"=Get Parent!(self:Material Instance Object Reference=@n5) -> (Parent:Material Interface Object Reference)
   ) {is=title K2Node_CallArrayFunction}
   Add(execute=@n5.CastFailed, TargetArray=@n6, NewItem=@n4)
   n7 = "Get (a copy)"(Array:Array of Material Interface Object References=@n6, Dimension 1=@n4.Array Index) -> (Output:Material Interface Object Reference)
   > PrintString(InString=Append(A="mat: ", B=GetDisplayName(Object=@n7)))
   Add!(
    execute=@n3.then_1
    TargetMap:Map of Static Mesh Component Object References to Chrome Mat Storage Struct Structures=Get cache()
    Key:"Static Mesh Component Object Reference (by ref)"=@n1
    Value:"Chrome Mat Storage Struct Structure (by ref)"=Make Chrome Mat Storage Struct!(materials_6_#1:Array of Material Interface Object References=Get l_mats()) -> (Chrome_MatStorage_Struct:Chrome Mat Storage Struct Structure) {is=title K2Node_MakeStruct}
   ) -> (then) {is=title K2Node_CallFunction}
  ]}
 EventGraph {nodes [
   Event Begin Play()
   Event Tick!() -> (OutputDelegate, then, DeltaSeconds:float)
  ]}
 IsCompValidForStore {nodes [
   n1 = IsCompValidForStore!() -> (then, comp:"Static Mesh Component Object Reference (by ref)") {is=title K2Node_FunctionEntry}
   > Set l_comp(l_comp:Static Mesh Component Object Reference=@n1) -> (Output_Get:Static Mesh Component Object Reference)
   > n2 = Branch(Condition=Get bIsEditorOnly!(self:Actor Component Object Reference=Get l_comp() -> (l_comp:Static Mesh Component Object Reference)) -> (bIsEditorOnly:bool))
   > PrintString(InString="     - Bad: Editor")
   > Return Node!(execute, ReturnValue:bool=false)
   n3 = Branch(execute=@n2.else, Condition=Find(TargetMap=Get cache(), Key=Get l_comp()).ReturnValue)
   > PrintString(InString="     - Bad: Exists")
   > Return Node(ReturnValue=false)
   n4 = Set l_mat(execute=@n3.else, l_mat:Array of Material Interface Object References=GetMaterials(self=Get l_comp())) -> (Output_Get:Array of Material Interface Object References)
   > n5 = For Each Loop(Array=@n4)
   Return Node(execute=@n5.Completed, ReturnValue=true)
   > n6 = GetBaseMaterial!(execute, self:Material Interface Object Reference=@n5) -> (then, ReturnValue:Material Object Reference)
   > Branch(Condition=Contains(SearchIn=GetDisplayName(Object=@n6), Substring=lackluster))
   > PrintString(InString="     - Bad: lackluster")
   > Return Node(ReturnValue=false)
  ]}
 StoreMaterials {nodes [
   StoreMaterials() {is=title K2Node_FunctionEntry}
   > n1 = For Each Loop(Array:Array of Static Mesh Component Object References=Get Components By Class!(self:Actor Object Reference=GetOwner(), ComponentClass:Actor Component Class Reference=&StaticMeshComponent) -> (ReturnValue:Array of Static Mesh Component Object References)) -> (Array Element:Static Mesh Component Object Reference)
   > PrintString(InString=Append(A="Comp ", B=GetDisplayName(Object=@n1)))
   > n2 = IsCompValidForStore!(execute, self, comp:"Static Mesh Component Object Reference (by ref)"=@n1) -> (then, ReturnValue:bool) {is=title K2Node_CallFunction}
   > Branch(Condition=@n2)
   > AddComponentToList!(execute, self, meshComponent:"Static Mesh Component Object Reference (by ref)"=@n1) -> (then) {is=title K2Node_CallFunction}
  ]}
 Unchrome {nodes [
   Unchrome() {is=title K2Node_FunctionEntry}
   > n1 = Sequence()
   > n2 = Keys!(execute, TargetMap:Map of Static Mesh Component Object References to Chrome Mat Storage Struct Structures=Get cache()) -> (then, Keys:Array of Static Mesh Component Object References)
   > n3 = For Each Loop(Array=@n2)
   > n4 = Set l_comp(l_comp=@n3)
   > Is Valid(InputObject=@n4)
   n5 = Get cache()
   > n6 = Set l_mat(l_mat=Break Chrome Mat Storage Struct!(Chrome_MatStorage_Struct:"Chrome Mat Storage Struct Structure (by ref)"=Find(TargetMap=@n5, Key=@n3)) -> (materials_6_#1:Array of Material Interface Object References) {is=title K2Node_BreakStruct})
   > n7 = For Each Loop(Array:Array of Material Interface Object References=@n6) -> (Array Element:Material Interface Object Reference)
   > SetMaterial(self=Get l_comp(), ElementIndex=@n7.Array Index, Material=@n7)
   Destroy Component!(execute=@n1.then_1, self:Actor Component Object Reference, Object:Object Reference) -> (then)
  ]}
]
variables [cache {}]
